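# Physics asset for the Anarchy Acres Farmer pickaxe, giving its hose bones simple capsule bodies for simulated motion.
# Two four-bone hose chains, A from engineBase and B from the two-handed handle, use limited angular joints, with twist and one swing axis locked at each root.
# The chain tips are joined by an angularly locked constraint, closing one hose loop between engine and handle.
# The engineBase-to-handle constraint leaves all angular axes free.
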PhysicsAsset /BRCosmetics/Weapons/FORT_Melee/Pickaxe_Anarchy_Acres_Farmer/Meshes/Anarchy_Acres_Farmer_Physics
bodies {
 Melee_TwoHanded_Handle [{shapeType=Capsule, length=82.2077, radius=3.2698, center=<1.3957 -1.4663 19.2663>, rotation=<rot -0.3908 -89.9998 -0.5336>}]
 dyn_hoseA1 [{shapeType=Capsule, length=1, radius=1, center=<0.0878 -0.0137 0>, rotation=<rot 90 7.125 27.6229>}]
 dyn_hoseA2 [{shapeType=Capsule, length=1, radius=1, center=<1.8509 0 -0.0001>, rotation=<rot 90 0 0>}]
 dyn_hoseA3 [{shapeType=Capsule, length=1, radius=1, center=<1.5822 -0.0283 0>, rotation=<rot 90 0 1.0231>}]
 dyn_hoseA4 [{shapeType=Capsule, length=1, radius=1, center=<0.9922 0.7339 -0.0001>, rotation=<rot 90 0 -55.5975>}]
 dyn_hoseB1 [{shapeType=Capsule, length=1, radius=1, center=<0.1902 -0.004 0>, rotation=<rot 90 0 -26.0972>}]
 dyn_hoseB2 [{shapeType=Capsule, length=1, radius=1, center=<1.4321 0 -0.0001>, rotation=<rot 90 0 0>}]
 dyn_hoseB3 [{shapeType=Capsule, length=1, radius=1, center=<1.0975 -0.0007 0>, rotation=<rot 90 -5.1339 -4.6222>}]
 dyn_hoseB4 [{shapeType=Capsule, length=1, radius=1, center=<1.0566 -0.8093 -0.0001>, rotation=<rot 90 180 -136.5555>}]
 engineBase [{shapeType=Capsule, length=0, radius=1.2667, center=<0 -7.8405 4.4974>, rotation=<rot 0 0 -90>}]
}
constraints [
 {
  bone1Name=engineBase
  bone2Name=Melee_TwoHanded_Handle
  swing1LimitDegrees=45
  swing1Motion=Free
  swing2LimitDegrees=45
  swing2Motion=Free
  twistLimitDegrees=45
  twistMotion=Free
 }
 {
  bone1Name=dyn_hoseA1
  bone2Name=engineBase
  swing1LimitDegrees=45
  swing2LimitDegrees=45
  swing2Motion=Locked
  twistLimitDegrees=15
  twistMotion=Locked
 }
 {bone1Name=dyn_hoseA2, bone2Name=dyn_hoseA1, swing1LimitDegrees=45, swing2LimitDegrees=45, twistLimitDegrees=15}
 {bone1Name=dyn_hoseA3, bone2Name=dyn_hoseA2, swing1LimitDegrees=45, swing2LimitDegrees=45, twistLimitDegrees=15}
 {bone1Name=dyn_hoseA4, bone2Name=dyn_hoseA3, swing1LimitDegrees=45, swing2LimitDegrees=45, twistLimitDegrees=15}
 {
  bone1Name=dyn_hoseB1
  bone2Name=Melee_TwoHanded_Handle
  swing1LimitDegrees=45
  swing2LimitDegrees=45
  swing2Motion=Locked
  twistLimitDegrees=15
  twistMotion=Locked
 }
 {bone1Name=dyn_hoseB2, bone2Name=dyn_hoseB1, swing1LimitDegrees=45, swing2LimitDegrees=45, twistLimitDegrees=15}
 {bone1Name=dyn_hoseB3, bone2Name=dyn_hoseB2, swing1LimitDegrees=45, swing2LimitDegrees=45, twistLimitDegrees=15}
 {bone1Name=dyn_hoseB4, bone2Name=dyn_hoseB3, swing1LimitDegrees=45, swing2LimitDegrees=45, twistLimitDegrees=15}
 {
  bone1Name=dyn_hoseB4
  bone2Name=dyn_hoseA4
  swing1LimitDegrees=45
  swing1Motion=Locked
  swing2LimitDegrees=45
  swing2Motion=Locked
  twistLimitDegrees=45
  twistMotion=Locked
 }
]
