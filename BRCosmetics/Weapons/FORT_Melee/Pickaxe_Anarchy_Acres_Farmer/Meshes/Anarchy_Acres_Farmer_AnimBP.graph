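# Anim Blueprint for the Anarchy Acres Farmer pickaxe that procedurally rattles its engine, exhausts and fork.
# Each update, sine waves of game time at differing frequencies set small translation offsets, plus pitch and roll for the fork.
# The AnimGraph adds these offsets to the input pose through additive bone-space Modify Bone nodes.
# A RigidBody node then simulates the override physics asset relative to the two-handed handle bone, with damped component motion.
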AnimBlueprint /BRCosmetics/Weapons/FORT_Melee/Pickaxe_Anarchy_Acres_Farmer/Meshes/Anarchy_Acres_Farmer_AnimBP
target_skeleton=&/Anarchy_Acres_Farmer_Skeleton
graphs [
 AnimGraph {nodes [
   AnimGraphNode_SubInput_0 = Input Pose()
   n1 = "Transform (Modify) Bone - Bone: exhaust_l"(
    ComponentPose=Local To Component(LocalPose=@AnimGraphNode_SubInput_0)
    Alpha=1
    Translation=Get exhaustLeft() -> (exhaustLeft:Vector)
    Rotation="0,0,0"
    Scale="1,1,1"
   ) {
    rotationMode=BMM_Additive
    rotationSpace=BCS_BoneSpace
    translationMode=BMM_Additive
    translationSpace=BCS_BoneSpace
    alphaBoolBlend {blendOption=Linear}
    boneToModify {boneName=exhaust_l}
   }
   n2 = "Transform (Modify) Bone - Bone: exhaust_r"(ComponentPose=@n1, Alpha=1, Translation=Get exhaustRight() -> (exhaustRight:Vector), Rotation="0,0,0", Scale="1,1,1") {
    rotationMode=BMM_Additive
    rotationSpace=BCS_BoneSpace
    translationMode=BMM_Additive
    translationSpace=BCS_BoneSpace
    alphaBoolBlend {blendOption=Linear}
    boneToModify {boneName=exhaust_r}
   }
   n3 = "Transform (Modify) Bone - Bone: fork"(
    ComponentPose=@n2
    Alpha=1
    Translation=Get forkTranslate() -> (forkTranslate:Vector)
    Rotation=Get forkRotate() -> (forkRotate:Rotator)
    Scale="1,1,1"
   ) {
    rotationMode=BMM_Additive
    rotationSpace=BCS_BoneSpace
    translationMode=BMM_Additive
    translationSpace=BCS_BoneSpace
    alphaBoolBlend {blendOption=Linear}
    boneToModify {boneName=fork}
   }
   n4 = "Transform (Modify) Bone - Bone: engineBase"(ComponentPose=@n3, Alpha=1, Translation=Get engine() -> (engine:Vector), Rotation="0,0,0", Scale="1,1,1") {
    rotationMode=BMM_Additive
    rotationSpace=BCS_BoneSpace
    translationMode=BMM_Additive
    translationSpace=BCS_BoneSpace
    alphaBoolBlend {blendOption=Linear}
    boneToModify {boneName=engineBase}
   }
   n5 = RigidBody(ComponentPose=@n4) {
    bClampLinearTranslationLimitToRefPose=true
    bOverrideWorldGravity=true
    overridePhysicsAsset=&/Anarchy_Acres_Farmer_Physics
    simulationSpace=BaseBoneSpace
    alphaBoolBlend {blendOption=Linear}
    baseBoneRef {boneName=Melee_TwoHanded_Handle}
    componentAppliedLinearAccClamp=<500 500 1000>
    componentLinearAccScale=<0.2 0.2 0.3>
    componentLinearVelScale=<0.1 0.1 0.1>
    overrideWorldGravity {z=-980}
   }
   Output Pose(Result=Component To Local(ComponentPose=@n5)) {layerGroup=None}
  ]}
 EventGraph {nodes [
   Event BlueprintUpdateAnimation()
   n1 = GetGameTimeInSeconds()
   n2 = "Sin (Radians)"(A=float * float(A=float * float(A=1, B=20), B=@n1))
   > Set exhaustLeft(exhaustLeft:Vector=MakeVector(X=float * float(A=@n2, B=0.15), Y=0)) -> (Output_Get:Vector)
   n3 = "Sin (Radians)"(A=float * float(A=float * float(A=1, B=15), B=@n1))
   > Set exhaustRight(exhaustRight:Vector=MakeVector(X=float * float(A=@n3, B=0.15), Y=0)) -> (Output_Get:Vector)
   n4 = "Sin (Radians)"(A=float * float(A=float * float(A=1, B=18), B=@n1))
   > Set forkTranslate(forkTranslate:Vector=MakeVector(X=float * float(A=@n4, B=0.2), Y=0)) -> (Output_Get:Vector)
   n5 = "Sin (Radians)"(A=float * float(A=float * float(A=1, B=40), B=@n1))
   > Set engine(engine:Vector=MakeVector(Y=0, Z=float * float(A=@n5, B=0.15))) -> (Output_Get:Vector)
   n6 = "Sin (Radians)"(A=float * float(A=float * float(A=1, B=10), B=@n1))
   n7 = "Sin (Radians)"(A=float * float(A=float * float(A=1, B=13), B=@n1))
   > Set forkRotate(forkRotate:Rotator=MakeRotator(Roll=float * float(A=@n7, B=0.5), Pitch=float * float(A=@n6, B=0.7))) -> (Output_Get:Rotator)
   TryGetPawnOwner()
  ]}
]
variables [
 engine {type=Vector}
 exhaustLeft {type=Vector}
 exhaustRight {type=Vector}
 forkRotate {type=Rotator}
 forkTranslate {type=Vector}
]
